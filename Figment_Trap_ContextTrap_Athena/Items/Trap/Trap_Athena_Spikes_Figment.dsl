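; Floor spike trap for Figment in Athena, handling the visuals and sounds of placement, firing, reload and destruction.
; Timelines drive the spike mesh's Z scale: the activation gameplay cue fires, then retracts after an arm-time delay, and reload end resets it.
; Always-active traps raise spikes on BeginPlay off dedicated servers and skip retraction.
; A dynamic material shows time-offset color sweeps, and a team check applies the friendly material.
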(event EventBeginPlay
  ;[1 1 1] Imported Traps do not reliably play other functions that control spike height/visibility
  (|CallParentFunction)
  (CallFunction|SetSpikePosition (Variables|Default|GetSpikeRestingScale))
  (Development|SetHiddeninGame (Variables|Trap_Floor_Spikes|GetSpike_Mesh) false false)
  ;;[0.01838 0.1354 0.15 0.5] Automatically lift the spikes if IsTrapAlwaysActive
  (if (and (Variables|Default|GetbIsTrapAlwaysActive) (not (Networking|IsDedicatedServer)))
    (CallFunction|TriggerTrap)))

(event EventOnPlaced
  (bind _returnvalue (Building|IsPreviewBuildingActor))
  ;[1 1 1] Initial Setup
  (Utilities|FlowControl|DoOnce false
    (:Completed
      (Trap|AddTriggerComponent (Variables|Trap_Floor_Spikes|GetTrigger))
      (if (not (Variables|Default|GetbPlayFewerSounds))
        (Audio|PlaySoundatLocation (Variables|Audio|GetPlace_Trap_Sound) (Transformation|GetActorLocation)))
      (CallFunction|SetSpikePosition (select _returnvalue 0.0 (Variables|Default|GetSpikeRestingScale)))
      (Development|SetHiddeninGame (Variables|Trap_Floor_Spikes|GetSpike_Mesh) _returnvalue false)
      (Utilities|ExTrapPlacementDustFX :Attach%20to%20Component (Variables|Debug|GetStaticMeshComponent) :Socket "PlacedFX" :Rotation (Variables|FX|GetTrap%20Particle%20Placement%20Rotation) :Burst%20Rate 15.0 :Size%20%28Volume%29 "200.000000,200.000000,15.000000" :Velocity%20Scale "1,1,1" :Size%20Scale%20%28Particles%29 1 :Particle%20Overrides (struct :FXType "Trap" :NiagaraVersion "/Game/Items/Traps/Particles/NS_Trap_Athena_Spikes.NS_Trap_Athena_Spikes" :CascadeVersion "/Game/Items/Traps/Particles/P_TrapPlacementDust_01.P_TrapPlacementDust_01")
        (:Out
          (Utilities|FlowControl|Delay 1.0)
          (Audio|PlaySoundatLocation (Variables|Audio|GetTrap_Active_Sound) (Transformation|GetActorLocation)))))))

(event EventOnFinishedBuilding
  ;[1 1 1] Finished building
  (Networking|CanExecuteCosmeticEvents
    (:True
      (CallFunction|SetSpikePosition (Variables|Default|GetSpikeRestingScale))
      (if (not (Variables|Default|GetbPlayFewerSounds))
        (Audio|PlaySoundatLocation (Variables|Audio|GetTrap_Active_Sound) (Transformation|GetActorLocation)))
      (Development|SetHiddeninGame (Variables|Trap_Floor_Spikes|GetSpike_Mesh) false false)
      (CallFunction|MaterialSweeponSpikes (Variables|Default|GetSweepSpeed) (Variables|Default|GetSweepColor) (Variables|Default|GetSweepExponent)))))

(event EventOnWorldReady
  ;[1 1 1] Finished building
  (Networking|CanExecuteCosmeticEvents
    (:True
      (CallFunction|SetSpikePosition (Variables|Default|GetSpikeRestingScale))
      (if (not (Variables|Default|GetbPlayFewerSounds))
        (Audio|PlaySoundatLocation (Variables|Audio|GetTrap_Active_Sound) (Transformation|GetActorLocation)))
      (Development|SetHiddeninGame (Variables|Trap_Floor_Spikes|GetSpike_Mesh) false false)
      (CallFunction|MaterialSweeponSpikes (Variables|Default|GetSweepSpeed) (Variables|Default|GetSweepColor) (Variables|Default|GetSweepExponent)))))

(event GameplayCue|GameplayCue.Abilities.Activation.Traps.ActivateTrap
  (params (EventType :type EGameplayCueEvent%20Enum :default "OnActive") (Parameters :type Gameplay%20Cue%20Parameters%20Structure%20%28by%20ref%29))
  (bind _returnvalue (Trap|GetArmTime))
  (bind _returnvalue_2 (- _returnvalue 1.0))
  ;[1 1 1] On Fire
  (switch Utilities|FlowControl|Switch|SwitchonEGameplayCueEvent EventType
    (:Executed
      (Audio|PlaySoundatLocation (Variables|Audio|GetTrap_Fire_Sound) (Transformation|GetActorLocation))
      ;; From this point on we don´t need to do this logic if the trap is always active (as we´ve done it already)
      (if (not (Variables|Default|GetbIsTrapAlwaysActive))
        (CallFunction|MaterialSweeponSpikes (Variables|Default|GetSweepSpeed) (Variables|Default|GetSweepColor) (Variables|Default|GetSweepExponent))
        (bind (_direction _lerp) (Timeline|Fire PlayFromStart)
          (:Update
            (CallFunction|SetSpikePosition (Math|Float|Lerp (Variables|Default|GetSpikeRestingScale) (Variables|Default|GetSpikeActiveScale) _lerp)))
          (:Finished
            (if (Variables|Default|GetbIsTrapAlwaysActive)
              _
              (elif (> _returnvalue_2 0)
                (Utilities|FlowControl|Delay _returnvalue_2)
                (bind (_direction_2 _lerp_2) (Timeline|Retract PlayFromStart)
                  (:Update
                    (CallFunction|SetSpikePosition (Math|Float|Lerp (Variables|Default|GetSpikeRestingScale) (Variables|Default|GetSpikeActiveScale) _lerp_2)))
                  (:Sounds
                    (Audio|PlaySoundatLocation (Variables|Audio|GetReload_Sound) (Transformation|GetActorLocation)) ; Reload sound
                  ))
                (else
                  _)))))))))

(event Custom|TriggerTrap
  (bind _returnvalue (Trap|GetArmTime))
  (bind _returnvalue_2 (- _returnvalue 1.0))
  ;[1 1 1] On Fire
  (CallFunction|MaterialSweeponSpikes (Variables|Default|GetSweepSpeed) (Variables|Default|GetSweepColor) (Variables|Default|GetSweepExponent))
  (Timeline|Fire PlayFromStart))

(event EventOnReloadEnd
  ;[1 1 1] Reload End
  (Audio|PlaySoundatLocation (Variables|Audio|GetTrap_Active_Sound) (Transformation|GetActorLocation))
  (Components|Timeline|Stop (Variables|Trap_Athena_Spikes_Figment|GetFire))
  (Development|SetHiddeninGame (Variables|Trap_Floor_Spikes|GetSpike_Mesh) false false)
  (CallFunction|ChangetoFriendlyMaterial)
  (CallFunction|MaterialSweeponSpikes (Variables|Default|GetSweepSpeed) (Variables|Default|GetSweepColor) (Variables|Default|GetSweepExponent))
  (bind (_direction _lerp) (Timeline|ResetSpikes PlayFromStart)
    (:Update
      (CallFunction|SetSpikePosition (Math|Float|Lerp 0.0 (Variables|Default|GetSpikeRestingScale) _lerp)))))

(event EventOnOutOfDurability
  ;[1 1 1] Durability Change
  (Utilities|FlowControl|Delay 1)
  (Audio|PlaySoundatLocation (Variables|Audio|GetTrap_Explode_Sound) (Transformation|GetActorLocation))
  (bind _asfxsystem_asset (Utilities|LoadRelevantParticleAsset (Variables|Default|GetDeathParticle_Marshalled))
    (:Loaded
      (Effects|Components|ParticleSystem|SpawnFXAtLocation :FXSystemAsset _asfxsystem_asset :Location (Transformation|GetActorLocation) :Scale "1.000000,1.000000,1.000000" :bAutoDestroy true :PoolingMethod "AutoRelease" :bAutoActivate true))))

(event EventOnInitTeam
  ;[1 1 1] Detect if local player is on the same team as this trap
  (CallFunction|ChangetoFriendlyMaterial))

(fn ChangeToFriendlyMaterial
  (bind _returnvalue (Fortnite|GetLocalFortPlayerControllers true))
  (if (and (Utilities|Enum|Equal(Enum) (Fortnite|GetSubGame) "Athena") (> (Utilities|Array|Length _returnvalue) 0))
    (if (Math|Byte|Equal(Byte) (Variables|Team|GetTeamIndex) (Team|GetActorTeam (Utilities|Array|Get(acopy) _returnvalue)))
      (Rendering|Material|SetMaterial (Variables|Debug|GetStaticMeshComponent) 0 (Variables|Default|GetFriendlyTrapMaterial)))))

(fn ChangeToNormalMaterial
  (Rendering|Material|SetMaterial (Variables|Debug|GetStaticMeshComponent) 0 (Variables|Default|GetNormalSpikeMaterial)))

(fn MaterialSweepOnSpikes
  (params (SweepSpeed :type Float%20%28double-precision%29 :default 6) (SweepColor :type Linear%20Color%20Structure :default "(R=0.260000,G=1.367674,B=4.000000,A=1.000000)") (SweepExponent :type Float%20%28double-precision%29 :default 80))
  (bind _result (Math|Float|NegateFloat (Utilities|Time|GetGameTimeinSeconds))
    (:%20%20
      (Rendering|Material|SetScalarParameterValue (Variables|Default|GetSpikesMID) "TimeStampOffset" _result :!class /Script/Engine.MaterialInstanceDynamic)
      (Rendering|Material|SetScalarParameterValue (Variables|Default|GetSpikesMID) "SweepSpeed" SweepSpeed :!class /Script/Engine.MaterialInstanceDynamic)
      (Rendering|Material|SetVectorParameterValue (Variables|Default|GetSpikesMID) "SweepColor" SweepColor :!class /Script/Engine.MaterialInstanceDynamic)
      (Rendering|Material|SetScalarParameterValue (Variables|Default|GetSpikesMID) "SweepExponent" SweepExponent :!class /Script/Engine.MaterialInstanceDynamic))))

(fn SetSpikePosition
  (params (NewSpikePosition :type Float%20%28double-precision%29))
  (Transformation|SetRelativeScale3D (Variables|Trap_Floor_Spikes|GetSpike_Mesh) (Math|Vector|MakeVector 1 1 NewSpikePosition)))

(fn ConstructionScript
  (bind _returnvalue (Rendering|Material|CreateDynamicMaterialInstance (Variables|Trap_Floor_Spikes|GetSpike_Mesh) 0 0 "None" :!class /Script/Engine.PrimitiveComponent))
  (Variables|Default|SetSpikesMID _returnvalue)
  (CallFunction|SetSpikePosition (Variables|Default|GetSpikeRestingScale)))
